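# Control Rig for the GrimeHold backpack that procedurally animates its upper and lower eye tentacles.
# Two chained ChainHarmonicsPerItem units add wave and pendulum motion from each tentacle root, with reach disabled and debug drawing enabled.
# Local-space-offset rotation constraints then tie tentacle tips and eye bases to the pack bone, main eye bones on X only, and secondary eye bones to weighted pack/main-eye blends.
# The aim bones are unused by the graph.
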ControlRigBlueprint /BRCosmetics/Accessories/FORT_Backpacks/Backpack_M_MED_GrimeHold/Meshes/M_MED_GrimeHold_Pack_CtrlRig
graphs [
 RigVMFunctionLibrary {links [], nodes []}
 RigVMModel {nodes [
   RigUnit_BeginExecution = &RigUnit_BeginExecution()
   /Script/ControlRig.RigUnit_ChainHarmonicsPerItem(
    ChainRoot=(Type=Bone,Name="base_tentacle_00_up")
    DrawWorldOffset=(Rotation=(X=0,Y=0,Z=0,W=1),Translation=(X=0,Y=0,Z=0),Scale3D=(X=1,Y=1,Z=1))
    Pendulum=(bEnabled=true,PendulumStiffness=2,PendulumGravity=(X=0,Y=0,Z=0),PendulumBlend=1,PendulumDrag=3,PendulumMinimum=0,PendulumMaximum=0.1,PendulumEase=CircularEaseInOut,UnwindAxis=(X=0,Y=1,Z=0),UnwindMinimum=0.2,UnwindMaximum=0.05)
    Reach=(bEnabled=false,ReachTarget=(X=0,Y=0,Z=0),ReachAxis=(X=1,Y=0,Z=0),ReachMinimum=0,ReachMaximum=0,ReachEase=SineEaseInOut)
    Speed=(X=0.5,Y=2,Z=1)
    Wave=(bEnabled=true,WaveFrequency=(X=90,Y=100,Z=7),WaveAmplitude=(X=0.03,Y=0.5,Z=0.3),WaveOffset=(X=0,Y=0,Z=0),WaveNoise=(X=0.3,Y=0.2,Z=0.5),WaveMinimum=-20,WaveMaximum=20,WaveEase=SineEaseInOut)
    WaveCurve=(EditorCurveData=(Keys=((InterpMode=RCIM_Cubic,Time=0.0055,Value=0.0453),(Time=1.0041,Value=0.6937))))
    bDrawDebug=True
    ExecutePin=@RigUnit_BeginExecution.ExecutePin
   ) {is=struct RigVMUnitNode}
   > &RigUnit_ChainHarmonicsPerItem(
    ChainRoot=(Type=Bone,Name="base_tentacle_01_dw")
    DrawWorldOffset=(Rotation=(X=0,Y=0,Z=0,W=1),Translation=(X=0,Y=0,Z=0),Scale3D=(X=1,Y=1,Z=1))
    Pendulum=(bEnabled=true,PendulumStiffness=2,PendulumGravity=(X=0,Y=0,Z=0),PendulumBlend=1,PendulumDrag=3,PendulumMinimum=0,PendulumMaximum=0.1,PendulumEase=CircularEaseInOut,UnwindAxis=(X=0,Y=1,Z=0),UnwindMinimum=0.2,UnwindMaximum=0.05)
    Reach=(bEnabled=false,ReachTarget=(X=0,Y=0,Z=0),ReachAxis=(X=1,Y=0,Z=0),ReachMinimum=0,ReachMaximum=0,ReachEase=SineEaseInOut)
    Speed=(X=0.5,Y=1,Z=1)
    Wave=(bEnabled=true,WaveFrequency=(X=20,Y=200,Z=7),WaveAmplitude=(X=0.07,Y=0.5,Z=1),WaveOffset=(X=0,Y=0,Z=0),WaveNoise=(X=2,Y=0.2,Z=0.5),WaveMinimum=-30,WaveMaximum=20,WaveEase=SineEaseInOut)
    WaveCurve=(EditorCurveData=(Keys=((InterpMode=RCIM_Cubic,Time=0.0055,Value=0.0453),(Time=1.0041,Value=0.6937))))
    bDrawDebug=True
   )
   > &RigUnit_RotationConstraintLocalSpaceOffset(
    AdvancedSettings=(InterpolationType=Average,RotationOrderForFilter=XZY)
    Child=(Type=Bone,Name="base_eye_tentacule_dw")
    Filter=(bX=True,bY=True,bZ=True)
    Parents=((Item=(Type=Bone,Name="pack"),Weight=1))
    Weight=1
   )
   > &RigUnit_RotationConstraintLocalSpaceOffset(
    AdvancedSettings=(InterpolationType=Average,RotationOrderForFilter=XZY)
    Child=(Type=Bone,Name="dyn_tentacle_05_dw")
    Filter=(bX=True,bY=True,bZ=True)
    Parents=((Item=(Type=Bone,Name="pack"),Weight=1))
    Weight=1
   )
   > &RigUnit_RotationConstraintLocalSpaceOffset(
    AdvancedSettings=(InterpolationType=Average,RotationOrderForFilter=XZY)
    Child=(Type=Bone,Name="dyn_eye_tentacule_dw1")
    Filter=(bX=True,bY=True,bZ=True)
    Parents=((Item=(Type=Bone,Name="pack"),Weight=0.7),(Item=(Type=Bone,Name="dyn_eye_tentacule_dw"),Weight=0.3))
    Weight=1
   )
   > &RigUnit_RotationConstraintLocalSpaceOffset(
    AdvancedSettings=(InterpolationType=Average,RotationOrderForFilter=XZY)
    Child=(Type=Bone,Name="dyn_eye_tentacule_dw2")
    Filter=(bX=True,bY=True,bZ=True)
    Parents=((Item=(Type=Bone,Name="pack"),Weight=0.7),(Item=(Type=Bone,Name="dyn_eye_tentacule_dw"),Weight=0.3))
    Weight=1
   )
   > &RigUnit_RotationConstraintLocalSpaceOffset(
    AdvancedSettings=(InterpolationType=Average,RotationOrderForFilter=XZY)
    Child=(Type=Bone,Name="dyn_eye_tentacule_dw")
    Filter=(bX=True,bY=false,bZ=false)
    Parents=((Item=(Type=Bone,Name="pack"),Weight=1))
    Weight=1
   )
   > &RigUnit_RotationConstraintLocalSpaceOffset(
    AdvancedSettings=(InterpolationType=Average,RotationOrderForFilter=XZY)
    Child=(Type=Bone,Name="base_eye_tentacule_up")
    Filter=(bX=True,bY=True,bZ=True)
    Parents=((Item=(Type=Bone,Name="pack"),Weight=1))
    Weight=1
   )
   > &RigUnit_RotationConstraintLocalSpaceOffset(
    AdvancedSettings=(InterpolationType=Average,RotationOrderForFilter=XZY)
    Child=(Type=Bone,Name="dyn_tentacle_05_up")
    Filter=(bX=True,bY=True,bZ=True)
    Parents=((Item=(Type=Bone,Name="pack"),Weight=1))
    Weight=1
   )
   > &RigUnit_RotationConstraintLocalSpaceOffset(
    AdvancedSettings=(InterpolationType=Average,RotationOrderForFilter=XZY)
    Child=(Type=Bone,Name="dyn_eye_tentacule_up1")
    Filter=(bX=True,bY=True,bZ=True)
    Parents=((Item=(Type=Bone,Name="pack"),Weight=0.6),(Item=(Type=Bone,Name="dyn_eye_tentacule_up"),Weight=0.4))
    Weight=1
   )
   > &RigUnit_RotationConstraintLocalSpaceOffset(
    AdvancedSettings=(InterpolationType=Average,RotationOrderForFilter=XZY)
    Child=(Type=Bone,Name="dyn_eye_tentacule_up2")
    Filter=(bX=True,bY=True,bZ=True)
    Parents=((Item=(Type=Bone,Name="pack"),Weight=0.6),(Item=(Type=Bone,Name="dyn_eye_tentacule_up"),Weight=0.4))
    Weight=1
   )
   > &RigUnit_RotationConstraintLocalSpaceOffset(
    AdvancedSettings=(InterpolationType=Average,RotationOrderForFilter=XZY)
    Child=(Type=Bone,Name="dyn_eye_tentacule_up")
    Filter=(bX=True,bY=false,bZ=false)
    Parents=((Item=(Type=Bone,Name="pack"),Weight=1))
    Weight=1
   )
  ]}
]
hierarchy [
 root {loc=<0 0 0>, rot=<0 0 0 1>}
 pack {parent=root, loc=<0 -0.7316 0>, rot=<0 0 0 1>}
 base_tentacle_01_dw {parent=pack, loc=<8.5663 -16.3321 2.2058>, rot=<0.4948 -0.3062 -0.0272 -0.8128>}
 dyn_tentacle_01_dw {parent=base_tentacle_01_dw, loc=<0 0 0>, rot=<0 0 0 1>}
 dyn_tentacle_02_dw {parent=dyn_tentacle_01_dw, loc=<3.7652 -0.0147 -0.1142>, rot=<0.0394 -0.6064 -0.0703 -0.7911>}
 dyn_tentacle_03_dw {parent=dyn_tentacle_02_dw, loc=<2.9392 1.1582 -0.2915>, rot=<-0.2716 -0.4636 0.3397 -0.772>}
 dyn_tentacle_04_dw {parent=dyn_tentacle_03_dw, loc=<5.1466 0.0446 -0.7304>, rot=<0.009 -0.0368 0.0802 -0.9961>}
 dyn_tentacle_05_dw {parent=dyn_tentacle_04_dw, loc=<4.3944 0.4972 0.2029>, rot=<-0.023 0.1896 -0.2401 -0.9518>}
 base_eye_tentacule_dw {parent=dyn_tentacle_05_dw, loc=<2.7578 0.961 0.2716>, rot=<-0.7331 -0.0114 -0.1115 -0.6708>}
 dyn_eye_tentacule_dw {parent=base_eye_tentacule_dw, loc=<0 0 0>, rot=<0 0 0 1>}
 dyn_eye_tentacule_dw2 {parent=base_eye_tentacule_dw, loc=<0 0 0>, rot=<0 0 0 1>}
 dyn_eye_tentacule_dw1 {parent=base_eye_tentacule_dw, loc=<0 0 0>, rot=<0 0 0 1>}
 dyn_eye_lowercrease_dw {parent=base_eye_tentacule_dw, loc=<0 0 0>, rot=<0 -0.3827 0 -0.9239>}
 dyn_eye_uppercrease_dw {parent=base_eye_tentacule_dw, loc=<0 0 0>, rot=<0 0.3827 0 -0.9239>}
 base_01 {parent=pack, loc=<-3.661 -17.8697 -17.7549>, rot=<-0.2108 -0.675 0.2108 -0.675>}
 dyn_01 {parent=base_01, loc=<0 0 0>, rot=<0 0 0 1>}
 base_02 {parent=pack, loc=<-5.4 -16.6757 -17.7198>, rot=<-0.2926 -0.6437 0.2926 -0.6437>}
 dyn_02 {parent=base_02, loc=<0 0 0>, rot=<0 0 0 1>}
 base_03 {parent=pack, loc=<-8.3338 -11.559 -17.7325>, rot=<0.5211 0.4779 -0.5211 0.4779>}
 dyn_03 {parent=base_03, loc=<0 0 0>, rot=<0 0 0 1>}
 base_cover {parent=pack, loc=<-3.2428 -10.51 7.5221>, rot=<-0.0187 0.3723 -0.0466 -0.9268>}
 dyn_cover {parent=base_cover, loc=<0 0 0>, rot=<0 0 0 1>}
 base_tentacle_00_up {parent=pack, loc=<0.5546 -9.2579 6.0658>, rot=<-0.5845 0.4973 -0.1532 -0.6225>}
 dyn_tentacle_00_up {parent=base_tentacle_00_up, loc=<0 0 0>, rot=<0 0 0 1>}
 dyn_tentacle_01_up {parent=dyn_tentacle_00_up, loc=<7.6779 0.1423 0.1411>, rot=<0.0511 0.0742 0.6166 -0.7821>}
 dyn_tentacle_02_up {parent=dyn_tentacle_01_up, loc=<3.8905 -0.4972 -0.5856>, rot=<0.0128 -0.0221 -0.184 -0.9826>}
 dyn_tentacle_03_up {parent=dyn_tentacle_02_up, loc=<4.0015 0.4389 -0.1926>, rot=<-0.0414 -0.0096 -0.6721 -0.7392>}
 dyn_tentacle_04_up {parent=dyn_tentacle_03_up, loc=<3.9103 -0.1662 -0.0679>, rot=<0.0699 0.0362 -0.3232 -0.943>}
 dyn_tentacle_05_up {parent=dyn_tentacle_04_up, loc=<5.0089 -0.1271 -0.0067>, rot=<0.0815 -0.0131 0.6291 -0.773>}
 base_eye_tentacule_up {parent=dyn_tentacle_05_up, loc=<3.718 -0.0179 0.4363>, rot=<0.7601 -0.1404 0.3021 -0.558>}
 dyn_eye_tentacule_up {parent=base_eye_tentacule_up, loc=<0 0 0>, rot=<0 0 0 1>}
 dyn_eye_tentacule_up1 {parent=base_eye_tentacule_up, loc=<0 0 0>, rot=<0 0 0 1>}
 dyn_eye_tentacule_up2 {parent=base_eye_tentacule_up, loc=<0 0 0>, rot=<0 0 0 1>}
 dyn_eye_lowercrease_up {parent=base_eye_tentacule_up, loc=<0 0 0>, rot=<0 -0.3827 0 -0.9239>}
 dyn_eye_uppercrease_up {parent=base_eye_tentacule_up, loc=<0 0 0>, rot=<0 0.3827 0 -0.9239>}
 aim_eye_tentacule_up {parent=pack, loc=<5 -50 10>, rot=<0 0 -0.5299 0.8481>}
 aim_eye_tentacule_dw {parent=pack, loc=<-5 -50 2.5532>, rot=<0 0 -0.8705 0.4922>}
]
