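# Animation Blueprint for the GrimeHold backpack cosmetic, adding secondary motion to its tentacles and blinking eyes.
# Anim Dynamics on three dyn bones feeds a cached pose, from which tentacles use Anim Dynamics while moving, jumping or ziplining, and a Control Rig otherwise.
# Additive bone transforms then blink both eyes, driven by randomly delayed alphas.
# The EventGraph also tracks zipline input events and computes a state-based gravity override.
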AnimBlueprint /BRCosmetics/Accessories/FORT_Backpacks/Backpack_M_MED_GrimeHold/Meshes/M_MED_GrimeHold_Pack_AnimBP
target_skeleton=&/M_MED_GrimeHold_Pack_Skeleton
graphs [
 AnimGraph {nodes [
   n1 = Get eye_up() -> (eye_up:double)
   n2 = "Transform (Modify) Bone - Bone: dyn_eye_tentacule_up1"(
    ComponentPose=Local To Component(LocalPose=Use cached pose 'system_cache'())
    Alpha=@n1
    Translation="0,0,0"
    Rotation="-50,0,0"
    Scale="1,1,1"
   ) {
    rotationMode=BMM_Additive
    rotationSpace=BCS_BoneSpace
    alphaBoolBlend {blendOption=Linear}
    alphaScaleBiasClamp {bInterpResult=true, interpSpeedDecreasing=15, interpSpeedIncreasing=15}
    boneToModify {boneName=dyn_eye_tentacule_up1}
    rotation {pitch=-50}
   }
   n3 = "Transform (Modify) Bone - Bone: dyn_eye_tentacule_up2"(ComponentPose=@n2, Alpha=@n1, Translation="0,0,0", Rotation="50,0,0", Scale="1,1,1") {
    rotationMode=BMM_Additive
    rotationSpace=BCS_BoneSpace
    alphaBoolBlend {blendOption=Linear}
    alphaScaleBiasClamp {bInterpResult=true, interpSpeedDecreasing=15, interpSpeedIncreasing=15}
    boneToModify {boneName=dyn_eye_tentacule_up2}
    rotation {pitch=50}
   }
   n4 = "Transform (Modify) Bone - Bone: dyn_eye_uppercrease_up"(ComponentPose=@n3, Alpha=@n1, Translation="0,-0.5,0", Rotation="0,0,0", Scale="1,1,1") {
    rotationSpace=BCS_BoneSpace
    translationMode=BMM_Additive
    alphaBoolBlend {blendOption=Linear}
    alphaScaleBiasClamp {bInterpResult=true, interpSpeedDecreasing=15, interpSpeedIncreasing=15}
    boneToModify {boneName=dyn_eye_uppercrease_up}
    translation {y=-0.5}
   }
   n5 = "Transform (Modify) Bone - Bone: dyn_eye_lowercrease_up"(ComponentPose=@n4, Alpha=@n1, Translation="0,-0.5,0", Rotation="0,0,0", Scale="1,1,1") {
    rotationSpace=BCS_BoneSpace
    translationMode=BMM_Additive
    alphaBoolBlend {blendOption=Linear}
    alphaScaleBiasClamp {bInterpResult=true, interpSpeedDecreasing=15, interpSpeedIncreasing=15}
    boneToModify {boneName=dyn_eye_lowercrease_up}
    translation {y=-0.5}
   }
   n6 = Get eye_dw() -> (eye_dw:double)
   n7 = "Transform (Modify) Bone - Bone: dyn_eye_tentacule_dw1"(ComponentPose=@n5, Alpha=@n6, Translation="0,0,0", Rotation="-45,0,0", Scale="1,1,1") {
    rotationMode=BMM_Additive
    rotationSpace=BCS_BoneSpace
    alphaBoolBlend {blendOption=Linear}
    alphaScaleBiasClamp {bInterpResult=true, interpSpeedDecreasing=15, interpSpeedIncreasing=15}
    boneToModify {boneName=dyn_eye_tentacule_dw1}
    rotation {pitch=-45}
   }
   n8 = "Transform (Modify) Bone - Bone: dyn_eye_tentacule_dw2"(ComponentPose=@n7, Alpha=@n6, Translation="0,0,0", Rotation="45,0,0", Scale="1,1,1") {
    rotationMode=BMM_Additive
    rotationSpace=BCS_BoneSpace
    alphaBoolBlend {blendOption=Linear}
    alphaScaleBiasClamp {bInterpResult=true, interpSpeedDecreasing=15, interpSpeedIncreasing=15}
    boneToModify {boneName=dyn_eye_tentacule_dw2}
    rotation {pitch=45}
   }
   n9 = "Transform (Modify) Bone - Bone: dyn_eye_uppercrease_dw"(ComponentPose=@n8, Alpha=@n6, Translation="0,-0.5,0", Rotation="0,0,0", Scale="1,1,1") {
    rotationSpace=BCS_BoneSpace
    translationMode=BMM_Additive
    alphaBoolBlend {blendOption=Linear}
    alphaScaleBiasClamp {bInterpResult=true, interpSpeedDecreasing=15, interpSpeedIncreasing=15}
    boneToModify {boneName=dyn_eye_uppercrease_dw}
    translation {y=-0.5}
   }
   n10 = "Transform (Modify) Bone - Bone: dyn_eye_lowercrease_dw"(ComponentPose=@n9, Alpha=@n6, Translation="0,-0.5,0", Rotation="0,0,0", Scale="1,1,1") {
    rotationSpace=BCS_BoneSpace
    translationMode=BMM_Additive
    alphaBoolBlend {blendOption=Linear}
    alphaScaleBiasClamp {bInterpResult=true, interpSpeedDecreasing=15, interpSpeedIncreasing=15}
    boneToModify {boneName=dyn_eye_lowercrease_dw}
    translation {y=-0.5}
   }
   Output Pose(Result=Component To Local(ComponentPose=@n10)) {layerGroup=None}
   n11 = Anim Dynamics - Bone: dyn_tentacle_05_up!(
    SimulationSpace:AnimPhysSimSpaceType Enum=World
    ComponentPose=Local To Component(LocalPose=Use cached pose 'liquid_cache'())
    Alpha=1
    ComponentLinearAccScale:Vector="0.8,0.8,0.8"
    ComponentLinearVelScale:Vector="0.8,0.8,0.8"
   ) -> (Pose) {
    angularDampingOverride=1
    angularSpringConstant=10000
    bAngularSpring=true
    bOverrideAngularDamping=true
    bUseGravityOverride=true
    gravityScale=0.35
    simulationSpace=World
    boundBone {boneName=dyn_tentacle_05_up}
    componentAppliedLinearAccClamp=<3000 3000 3000>
    componentLinearAccScale=<0.8 0.8 0.8>
    componentLinearVelScale=<0.8 0.8 0.8>
    gravityOverride {z=-980}
    physicsBodyDefinitions [{
      boundBone {boneName=dyn_tentacle_05_up}
      boxExtents=<30 30 30>
      constraintSetup {angularLimitsMax=<40 30 60>, angularLimitsMin=<-40 -30 -40>, angularTarget=<1 0 0>}
      localJointOffset=<4 0 0>
     }]
   }
   n12 = Anim Dynamics - Bone: dyn_tentacle_04_up!(
    SimulationSpace:AnimPhysSimSpaceType Enum=World
    ComponentPose=@n11
    Alpha=1
    ComponentLinearAccScale:Vector="0.3,0.3,0.3"
    ComponentLinearVelScale:Vector="0.2,0.2,0.2"
   ) -> (Pose) {
    angularDampingOverride=1
    angularSpringConstant=15000
    bAngularSpring=true
    bOverrideAngularDamping=true
    bUseGravityOverride=true
    gravityScale=0.35
    simulationSpace=World
    boundBone {boneName=dyn_tentacle_04_up}
    componentAppliedLinearAccClamp=<3000 3000 3000>
    componentLinearAccScale=<0.3 0.3 0.3>
    componentLinearVelScale=<0.2 0.2 0.2>
    gravityOverride {z=-980}
    physicsBodyDefinitions [{
      boundBone {boneName=dyn_tentacle_04_up}
      boxExtents=<30 30 30>
      constraintSetup {angularLimitsMax=<15 18 60>, angularLimitsMin=<-15 -18 -20>, angularTarget=<1 0 0>}
      localJointOffset=<4 0 0>
     }]
   }
   n13 = Anim Dynamics - Bone: dyn_tentacle_03_up!(
    SimulationSpace:AnimPhysSimSpaceType Enum=World
    ComponentPose=@n12
    Alpha=1
    ComponentLinearAccScale:Vector="0.3,0.3,0.3"
    ComponentLinearVelScale:Vector="0.2,0.2,0.2"
   ) -> (Pose) {
    angularDampingOverride=1
    angularSpringConstant=20000
    bAngularSpring=true
    bOverrideAngularDamping=true
    bUseGravityOverride=true
    gravityScale=0.35
    simulationSpace=World
    boundBone {boneName=dyn_tentacle_03_up}
    componentAppliedLinearAccClamp=<3000 3000 3000>
    componentLinearAccScale=<0.3 0.3 0.3>
    componentLinearVelScale=<0.2 0.2 0.2>
    gravityOverride {z=-980}
    physicsBodyDefinitions [{
      boundBone {boneName=dyn_tentacle_03_up}
      boxExtents=<30 30 30>
      constraintSetup {angularLimitsMax=<15 15 20>, angularLimitsMin=<-15 -15 0>, angularTarget=<1 0 0>}
      localJointOffset=<4 0 0>
     }]
   }
   n14 = Anim Dynamics - Bone: dyn_tentacle_02_up!(
    SimulationSpace:AnimPhysSimSpaceType Enum=World
    ComponentPose=@n13
    Alpha=1
    ComponentLinearAccScale:Vector="0.3,0.3,0.3"
    ComponentLinearVelScale:Vector="0.2,0.2,0.2"
   ) -> (Pose) {
    angularDampingOverride=1
    angularSpringConstant=25000
    bAngularSpring=true
    bOverrideAngularDamping=true
    bUseGravityOverride=true
    gravityScale=0.35
    simulationSpace=World
    boundBone {boneName=dyn_tentacle_02_up}
    componentAppliedLinearAccClamp=<3000 3000 3000>
    componentLinearAccScale=<0.3 0.3 0.3>
    componentLinearVelScale=<0.2 0.2 0.2>
    gravityOverride {z=-980}
    physicsBodyDefinitions [{
      boundBone {boneName=dyn_tentacle_02_up}
      boxExtents=<30 30 30>
      constraintSetup {angularLimitsMax=<18 20 45>, angularLimitsMin=<-18 -20 0>, angularTarget=<1 0 0>}
      localJointOffset=<4 0 0>
     }]
   }
   n15 = Anim Dynamics - Bone: dyn_tentacle_01_up!(
    SimulationSpace:AnimPhysSimSpaceType Enum=World
    ComponentPose=@n14
    Alpha=1
    ComponentLinearAccScale:Vector="0.3,0.3,0.3"
    ComponentLinearVelScale:Vector="0.2,0.2,0.2"
   ) -> (Pose) {
    angularDampingOverride=1
    angularSpringConstant=5000
    bAngularSpring=true
    bOverrideAngularDamping=true
    bUseGravityOverride=true
    gravityScale=0.35
    simulationSpace=World
    boundBone {boneName=dyn_tentacle_01_up}
    componentAppliedLinearAccClamp=<3000 3000 3000>
    componentLinearAccScale=<0.3 0.3 0.3>
    componentLinearVelScale=<0.2 0.2 0.2>
    gravityOverride {z=-980}
    physicsBodyDefinitions [{
      boundBone {boneName=dyn_tentacle_01_up}
      boxExtents=<30 30 30>
      constraintSetup {angularLimitsMax=<20 25 60>, angularLimitsMin=<-20 -25 0>, angularTarget=<1 0 0>}
      localJointOffset=<4 0 0>
     }]
   }
   n16 = Anim Dynamics - Bone: dyn_tentacle_05_dw!(
    SimulationSpace:AnimPhysSimSpaceType Enum
    ComponentPose=@n15
    Alpha=1
    ComponentLinearAccScale:Vector="0.3,0.3,0.3"
    ComponentLinearVelScale:Vector="0.2,0.2,0.2"
   ) -> (Pose) {
    angularDampingOverride=1
    angularSpringConstant=5000
    bAngularSpring=true
    bOverrideAngularDamping=true
    bUseGravityOverride=true
    gravityScale=0.35
    simulationSpace=RootRelative
    boundBone {boneName=dyn_tentacle_05_dw}
    componentAppliedLinearAccClamp=<3000 3000 3000>
    componentLinearAccScale=<0.3 0.3 0.3>
    componentLinearVelScale=<0.2 0.2 0.2>
    gravityOverride {z=-980}
    physicsBodyDefinitions [{
      boundBone {boneName=dyn_tentacle_05_dw}
      boxExtents=<30 30 30>
      constraintSetup {angularLimitsMax=<-20 5 20>, angularLimitsMin=<-40 -10 -20>, angularTarget=<1 0 0>}
      localJointOffset=<4 0 0>
     }]
   }
   n17 = Anim Dynamics - Bone: dyn_tentacle_04_dw!(
    SimulationSpace:AnimPhysSimSpaceType Enum
    ComponentPose=@n16
    Alpha=1
    ComponentLinearAccScale:Vector="0.3,0.3,0.3"
    ComponentLinearVelScale:Vector="0.2,0.2,0.2"
   ) -> (Pose) {
    angularDampingOverride=1
    angularSpringConstant=5000
    bAngularSpring=true
    bOverrideAngularDamping=true
    bUseGravityOverride=true
    gravityScale=0.35
    simulationSpace=RootRelative
    boundBone {boneName=dyn_tentacle_04_dw}
    componentAppliedLinearAccClamp=<3000 3000 3000>
    componentLinearAccScale=<0.3 0.3 0.3>
    componentLinearVelScale=<0.2 0.2 0.2>
    gravityOverride {z=-980}
    physicsBodyDefinitions [{
      boundBone {boneName=dyn_tentacle_04_dw}
      boxExtents=<30 30 30>
      constraintSetup {angularLimitsMax=<-20 20 15>, angularLimitsMin=<-30 -20 -15>, angularTarget=<1 0 0>}
      localJointOffset=<4 0 0>
     }]
   }
   n18 = Anim Dynamics - Bone: dyn_tentacle_03_dw!(
    SimulationSpace:AnimPhysSimSpaceType Enum
    ComponentPose=@n17
    Alpha=1
    ComponentLinearAccScale:Vector="0.3,0.3,0.3"
    ComponentLinearVelScale:Vector="0.2,0.2,0.2"
   ) -> (Pose) {
    angularDampingOverride=1
    angularSpringConstant=5000
    bAngularSpring=true
    bOverrideAngularDamping=true
    bUseGravityOverride=true
    gravityScale=0.35
    simulationSpace=RootRelative
    boundBone {boneName=dyn_tentacle_03_dw}
    componentAppliedLinearAccClamp=<3000 3000 3000>
    componentLinearAccScale=<0.3 0.3 0.3>
    componentLinearVelScale=<0.2 0.2 0.2>
    gravityOverride {z=-980}
    physicsBodyDefinitions [{
      boundBone {boneName=dyn_tentacle_03_dw}
      boxExtents=<30 30 30>
      constraintSetup {angularLimitsMax=<-10 0 20>, angularLimitsMin=<-30 0 -20>, angularTarget=<1 0 0>}
      localJointOffset=<4 0 0>
     }]
   }
   n19 = Anim Dynamics - Bone: dyn_tentacle_02_dw!(
    SimulationSpace:AnimPhysSimSpaceType Enum
    ComponentPose=@n18
    Alpha=1
    ComponentLinearAccScale:Vector="0.3,0.3,0.3"
    ComponentLinearVelScale:Vector="0.2,0.2,0.2"
   ) -> (Pose) {
    angularDampingOverride=1
    angularSpringConstant=5000
    bAngularSpring=true
    bOverrideAngularDamping=true
    bUseGravityOverride=true
    gravityScale=0.35
    simulationSpace=RootRelative
    boundBone {boneName=dyn_tentacle_02_dw}
    componentAppliedLinearAccClamp=<3000 3000 3000>
    componentLinearAccScale=<0.3 0.3 0.3>
    componentLinearVelScale=<0.2 0.2 0.2>
    gravityOverride {z=-980}
    physicsBodyDefinitions [{
      boundBone {boneName=dyn_tentacle_02_dw}
      boxExtents=<30 30 30>
      constraintSetup {angularLimitsMax=<15 0 0>, angularLimitsMin=<-15 0 -10>, angularTarget=<1 0 0>}
      localJointOffset=<4 0 0>
     }]
   }
   n20 = Anim Dynamics - Bone: dyn_tentacle_01_dw!(
    SimulationSpace:AnimPhysSimSpaceType Enum
    ComponentPose=@n19
    Alpha=1
    ComponentLinearAccScale:Vector="0.3,0.3,0.3"
    ComponentLinearVelScale:Vector="0.2,0.2,0.2"
   ) -> (Pose) {
    angularDampingOverride=1
    angularSpringConstant=5000
    bAngularSpring=true
    bOverrideAngularDamping=true
    bUseGravityOverride=true
    gravityScale=0.35
    simulationSpace=RootRelative
    boundBone {boneName=dyn_tentacle_01_dw}
    componentAppliedLinearAccClamp=<3000 3000 3000>
    componentLinearAccScale=<0.3 0.3 0.3>
    componentLinearVelScale=<0.2 0.2 0.2>
    gravityOverride {z=-980}
    physicsBodyDefinitions [{
      boundBone {boneName=dyn_tentacle_01_dw}
      boxExtents=<30 30 30>
      constraintSetup {angularLimitsMax=<15 20 20>, angularLimitsMin=<-15 0 -20>, angularTarget=<1 0 0>}
      localJointOffset=<4 0 0>
     }]
   }
   n21 = Control Rig(Source=Use cached pose 'liquid_cache'()) {
    alphaBoolBlend {blendOption=Linear}
    controlRigAssetReference {blueprintRigClass=&/M_MED_GrimeHold_Pack_CtrlRig.M_MED_GrimeHold_Pack_CtrlRig_C}
   }
   Save cached pose 'system_cache'(Pose=Blend Poses by bool(
     BlendPose_0=Component To Local(ComponentPose=@n20)
     BlendPose_1=@n21
     BlendTime_1=0.5
     bActiveValue=Get Check_Status() -> (Check_Status:bool)
    )) {cacheName=system_cache}
   n22 = Anim Dynamics - Bone: dyn_03!(ComponentPose=Local To Component(LocalPose=Input Pose()), Alpha=1, GravityOverride:Vector="0,0,-980", ExternalForce="0,0,0") -> (Pose) {
    angularDampingOverride=1
    angularSpringConstant=5000
    bAngularSpring=true
    bOverrideAngularDamping=true
    bUseGravityOverride=true
    gravityScale=0.35
    simulationSpace=World
    boundBone {boneName=dyn_03}
    componentAppliedLinearAccClamp=<3000 3000 3000>
    componentLinearAccScale=<0.2 0.2 0.2>
    componentLinearVelScale=<0.2 0.2 0.2>
    gravityOverride {z=-980}
    physicsBodyDefinitions [{
      boundBone {boneName=dyn_03}
      boxExtents=<25 25 25>
      constraintSetup {angularLimitsMax=<0 45 45>, angularLimitsMin=<0 -45 -45>, angularTarget=<1 0 0>}
      localJointOffset=<3 0 0>
     }]
   }
   n23 = Anim Dynamics - Bone: dyn_02!(ComponentPose=@n22, Alpha=1, GravityOverride:Vector="0,0,-980", ExternalForce="0,0,0") -> (Pose) {
    angularDampingOverride=1
    angularSpringConstant=5000
    bAngularSpring=true
    bOverrideAngularDamping=true
    bUseGravityOverride=true
    gravityScale=0.35
    simulationSpace=World
    boundBone {boneName=dyn_02}
    componentAppliedLinearAccClamp=<3000 3000 3000>
    componentLinearAccScale=<0.2 0.2 0.2>
    componentLinearVelScale=<0.2 0.2 0.2>
    gravityOverride {z=-980}
    physicsBodyDefinitions [{
      boundBone {boneName=dyn_02}
      boxExtents=<25 25 25>
      constraintSetup {angularLimitsMax=<0 45 45>, angularLimitsMin=<0 -45 -45>, angularTarget=<1 0 0>}
      localJointOffset=<3 0 0>
     }]
   }
   n24 = Anim Dynamics - Bone: dyn_01!(ComponentPose=@n23, Alpha=1, GravityOverride:Vector="0,0,-980", ExternalForce="0,0,0") -> (Pose) {
    angularDampingOverride=1
    angularSpringConstant=5000
    bAngularSpring=true
    bOverrideAngularDamping=true
    bUseGravityOverride=true
    gravityScale=0.35
    simulationSpace=World
    boundBone {boneName=dyn_01}
    componentAppliedLinearAccClamp=<3000 3000 3000>
    componentLinearAccScale=<0.2 0.2 0.2>
    componentLinearVelScale=<0.2 0.2 0.2>
    gravityOverride {z=-980}
    physicsBodyDefinitions [{
      boundBone {boneName=dyn_01}
      boxExtents=<25 25 25>
      constraintSetup {angularLimitsMax=<0 45 45>, angularLimitsMin=<0 -45 -45>, angularTarget=<1 0 0>}
      localJointOffset=<3 0 0>
     }]
   }
   Save cached pose 'liquid_cache'(Pose=Component To Local(ComponentPose=@n24)) {cacheName=liquid_cache}
  ]}
 EventGraph {nodes [
   Event BlueprintUpdateAnimation()
   > n1 = Sequence!(execute) -> (then_0, then_1, then_2:Exec, then_3:Exec, then_4:Exec, then_5:Exec)
   > n2 = "Gravity Override State Function (Parameter Asset) (Nativized)"(
    CharacterPartAnimInstance=Self-Reference()
    GameTimeInSeconds=GetGameTimeInSeconds()
    Parameters=&/M_MED_GrimeHold_Pack_Paramet
   )
   > Set GravityOverride(GravityOverride=@n2)
   Delay(execute=@n1.then_1, Duration=RandomFloatInRange(Min=4, Max=10))
   > Set eye_up(eye_up:double=1) -> (Output_Get:double)
   > Delay(Duration=RandomFloatInRange(Min=0.2, Max=0.4))
   > Set eye_up(eye_up=0)
   Set Check_Status(
    execute=@n1.then_2
    Check_Status:bool=OR Boolean!(A=Get bIsMoving2D(), B=Get bIsJumping(), C:bool=Get OnZipline() -> (OnZipline:bool)) -> (ReturnValue)
   ) -> (Output_Get:bool)
   Delay(execute=@n1.then_3, Duration=RandomFloatInRange(Min=3, Max=8))
   > Set eye_dw(eye_dw:double=1) -> (Output_Get:double)
   > Delay(Duration=RandomFloatInRange(Min=0.2, Max=0.4))
   > Set eye_dw(eye_dw=0)
   n3 = Get bIsInFrontEnd()
   Set Dyn Alpha(execute=@n1.then_4, Dyn Alpha:bool=NOT Boolean(A=@n3)) -> (Output_Get:bool)
   > n4 = Branch(Condition=@n3)
   > Set alpha frontend(alpha frontend:double=0.5) -> (Output_Get:double)
   > Set simspace var(simspace var:Byte=2) -> (Output_Get:Byte=0)
   Set alpha frontend(execute=@n4.else, alpha frontend=1)
   > Set simspace var(simspace var=4) -> (Output_Get=0)
   n5 = Event BlueprintProcessAnimInputEvent()
   > n6 = Branch(Condition=IsNamed(self=@n5.AnimInputEvent, InEventName=aie.Player.Zipline.Entered))
   > Set OnZipline(OnZipline:bool=true) -> (Output_Get:bool)
   Branch(execute=@n6.else, Condition=IsNamed(self=@n5.AnimInputEvent, InEventName=aie.Player.Zipline.Exited))
   > Set OnZipline(OnZipline=false)
   TryGetPawnOwner()
  ]}
]
variables [
 alpha frontend {type=number}
 animDyn on zipline {type=Vector, default=<0.5 0.5 0.5>}
 Check_Status {type=boolean}
 Dyn Alpha {type=boolean}
 eye_dw {type=number}
 eye_up {type=number}
 Gravity Override {type=Vector, default=<0 0 -980>}
 Gravity Override Cape {type=Vector}
 Gravity Override Params_0 {
  type=FortGravityOverrideParameterSet
  default=(Skydive_GoingUp=(GravityOverride=()),Skydive_FloatingDown=(GravityOverride=()),Skydive_DivingDown=(GravityOverride=()),Skydive_Parachuting=(GravityOverride=()),OnGround_Standing=(GravityOverride=()),OnGround_Moving=(GravityOverride=()),OnGround_Moving_HighRotationRate=(GravityOverride=()),OnGround_Crouch_Moving=(GravityOverride=()),Emote_or_Melee=(GravityOverride=()),Swimming=(GravityOverride=()),Front_End=(GravityOverride=()),DBNO=(GravityOverride=()),DBNO_Carried=(GravityOverride=()),OnGround_Falling=(GravityOverride=()),Ziplining=(GravityOverride=()),Skydive_Surfing=(GravityOverride=()),Riding_Standing=(GravityOverride=()),Riding_Moving=(GravityOverride=()),Riding_Moving_Targeting=(GravityOverride=()),Tactical_Sprint=(GravityOverride=()),Rail_Grinding=(GravityOverride=()),)
 }
 GravityOverride {type=Vector}
 OnZipline {type=boolean}
 RBAN Params_0 {
  type=FortRigidBodyAnimNodeParameterSet
  default=(Skydive_GoingUp=(RigidBodyAnimNode=()),Skydive_FloatingDown=(RigidBodyAnimNode=()),Skydive_DivingDown=(RigidBodyAnimNode=()),Skydive_Parachuting=(RigidBodyAnimNode=()),OnGround_Standing=(RigidBodyAnimNode=()),OnGround_Moving=(RigidBodyAnimNode=()),OnGround_Moving_HighRotationRate=(RigidBodyAnimNode=()),OnGround_Crouch_Moving=(RigidBodyAnimNode=()),Emote_or_Melee=(RigidBodyAnimNode=()),Swimming=(RigidBodyAnimNode=()),Front_End=(RigidBodyAnimNode=()),DBNO=(RigidBodyAnimNode=()),DBNO_Carried=(RigidBodyAnimNode=()),OnGround_Falling=(RigidBodyAnimNode=()),Ziplining=(RigidBodyAnimNode=()),Skydive_Surfing=(RigidBodyAnimNode=()),Riding_Standing=(RigidBodyAnimNode=()),Riding_Moving=(RigidBodyAnimNode=()),Riding_Moving_Targeting=(RigidBodyAnimNode=()),Tactical_Sprint=(RigidBodyAnimNode=()),Rail_Grinding=(RigidBodyAnimNode=()),)
 }
 RBAN Settings Out {type=FortRigidBodyAnimNodeParameters}
 RBAN Settings Out_0 {type=FortRigidBodyAnimNodeParameters}
 simspace var {type=integer}
 zipline_joint_05 {type=Vector, default=<0.5 0.5 0.5>}
]
